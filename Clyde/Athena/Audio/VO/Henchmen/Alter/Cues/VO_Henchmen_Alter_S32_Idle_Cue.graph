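# Sound cue that plays idle voice lines for the Alter henchman NPCs.
# A single Random node picks among ten equally weighted wave players, each holding one idle line variation.
# It uses shared close-range NPC VO attenuation, a non-critical enemy VO sound class and the henchmen combat chatter concurrency, with reduced volume.
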SoundCue /Clyde/Athena/Audio/VO/Henchmen/Alter/Cues/VO_Henchmen_Alter_S32_Idle_Cue
root=Random_0
nodes [
 {
  id=Random_0
  children [WavePlayer_0,WavePlayer_1,WavePlayer_2,WavePlayer_3,WavePlayer_4,WavePlayer_5,WavePlayer_6,WavePlayer_7,WavePlayer_8,WavePlayer_9]
  properties {weights [1,1,1,1,1,1,1,1,1,1]}
 }
 {id=WavePlayer_0, properties {soundWaveAssetPtr=/Clyde/Athena/Audio/VO/Henchmen/Alter/Waves/VO_Henchmen_Alter_S32_Idle_010}}
 {id=WavePlayer_1, properties {soundWaveAssetPtr=/Clyde/Athena/Audio/VO/Henchmen/Alter/Waves/VO_Henchmen_Alter_S32_Idle_020}}
 {id=WavePlayer_2, properties {soundWaveAssetPtr=/Clyde/Athena/Audio/VO/Henchmen/Alter/Waves/VO_Henchmen_Alter_S32_Idle_030}}
 {id=WavePlayer_3, properties {soundWaveAssetPtr=/Clyde/Athena/Audio/VO/Henchmen/Alter/Waves/VO_Henchmen_Alter_S32_Idle_040}}
 {id=WavePlayer_4, properties {soundWaveAssetPtr=/Clyde/Athena/Audio/VO/Henchmen/Alter/Waves/VO_Henchmen_Alter_S32_Idle_050}}
 {id=WavePlayer_5, properties {soundWaveAssetPtr=/Clyde/Athena/Audio/VO/Henchmen/Alter/Waves/VO_Henchmen_Alter_S32_Idle_060}}
 {id=WavePlayer_6, properties {soundWaveAssetPtr=/Clyde/Athena/Audio/VO/Henchmen/Alter/Waves/VO_Henchmen_Alter_S32_Idle_070}}
 {id=WavePlayer_7, properties {soundWaveAssetPtr=/Clyde/Athena/Audio/VO/Henchmen/Alter/Waves/VO_Henchmen_Alter_S32_Idle_080}}
 {id=WavePlayer_8, properties {soundWaveAssetPtr=/Clyde/Athena/Audio/VO/Henchmen/Alter/Waves/VO_Henchmen_Alter_S32_Idle_090}}
 {id=WavePlayer_9, properties {soundWaveAssetPtr=/Clyde/Athena/Audio/VO/Henchmen/Alter/Waves/VO_Henchmen_Alter_S32_Idle_100}}
]
properties {
 attenuationSettings=/Game/Sounds/Fort_Human_VO_BR/NPCs/Shared/NPC_VO_Close_Att
 maxDistance=8500
 soundClassObject=/Game/Sounds/FortSoundClassesAndModes/Fort_VO_Enemy_NonCritical
 volumeMultiplier=0.4743
 concurrencySet [/Game/Sounds/Fort_Human_VO_BR/NPCs/Shared/NPC_Henchmen_Combat_Chatter_Conc]
}
